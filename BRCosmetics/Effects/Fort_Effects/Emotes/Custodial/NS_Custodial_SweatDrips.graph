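# Niagara system for sweat drips in the Custodial emote, with one sprite emitter that burst-spawns a single drip by chance.
# Curve-driven gravity keeps each drip slow at first, then pulls it down sharply, while the sprite narrows and stretches.
# Two sprite renderers are present, and separate DripSize and DripColor attributes allow a crossfade: DripColor alpha rises as main Color fades out.
# Spawn rate, start offset and rotated initial velocity are disabled.
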NiagaraSystem /BRCosmetics/Effects/Fort_Effects/Emotes/Custodial/NS_Custodial_SweatDrips
dependencies {
 usedDataInterfaces [
  &NiagaraDataInterfaceVectorCurve
  &NiagaraDataInterfaceVector2DCurve
  &NiagaraDataInterfaceColorCurve
 ]
 usedDynamicInputs []
 usedModules [&SystemState]
}
emitters {Drip1 {
  emitterUpdateScript [
   &EmitterState(
    Life Cycle Mode=::NewEnumerator1
    Loop Delay=&UniformRangedFloat(Minimum=0.5, Maximum=1)
    Loop Duration=1
    UseLoopDelay=false
   )
   &SpawnRate(Use Spawn Probability=false) {enabled=false}
   &SpawnBurst_Instantaneous(Spawn Count=1, Spawn Probability=0.3, Use Spawn Probability=true)
  ]
  particleSpawnScript [
   /Niagara/Modules/Spawn/Initialization/InitializeParticle(
    Lifetime=&RandomRangeFloat(Minimum=0.6, Maximum=0.8)
    Mass=&UniformRangedFloat(Minimum=0.75, Maximum=1.75)
    Module.Write Color=true
    Module.Write Lifetime=true
    Module.Write Mass=false
    Module.Write Material Random=true
    Module.Write Position=false
    Module.Write Scale=false
    Module.Write SpriteRotation=true
    Module.Write SpriteSize=true
    Position=&SimulationPosition()
    Sprite Rotation=180
    Sprite Size=&Vector2DFromFloat(Value=&RandomRangeFloat(Minimum=2, Maximum=3.5))
   )
   /SetVariables(Particles.DripSize:Vector2f=$Particles.SpriteSize) {isSetParameters=true}
   /SetVariables(Particles.StartPositionOffeset:Vector3f=&RandomRangeVector(Minimum=<-0.5 -0.5 -4>, Maximum=<0.5 0.5 -6>)) {enabled=false, isSetParameters=true}
   &AddVelocity(Velocity=&RotateVectorByQuaternion(Vector To Rotate=<0 0 -1>, Quaternion=$Output.SkeletalMeshLocation001.SampledRotation)) {enabled=false}
  ]
  particleUpdateScript [
   &GravityForce(Gravity=&VectorFromCurve(
     VectorCurve=DataInterface() {propertyValues {XCurve {keys [<at 0 2>,<at 1 0>]}, YCurve {keys [<at 0 0>,<at 1 0>]}, ZCurve {keys [<at 0 0>,<at 0.4222 10>,<at 0.5637 140>,<at 1 90>]}}}
     Scale Curve=&RandomRangeVector(Minimum=<0 0 -1.75>, Maximum=<0 0 -2.25>)
    ))
   &InheritVelocity(Module.Limit Speed=true, Velocity Limit=0.1)
   &SolveForcesAndVelocity(Clamp Velocity=false, Limit Acceleration=false)
   &ParticleState()
   &CameraOffset(Camera Offset Amount=&RandomRangeFloat(Minimum=-2, Maximum=5))
   /SetVariables(Particles.DripSize:Vector2f=&Multiply_Vector2(
     A=$Particles.Initial.DripSize
     B=&Vector2DFromCurve(Vector2Curve=DataInterface() {propertyValues {XCurve {keys [<at 0 1>,<at 1 1>]}, YCurve {keys [<at 0 4>,<at 1 1>]}}}, ScaleCurve=<1 1>)
    )) {isSetParameters=true}
   /SetVariables(Particles.DripColor:LinearColor=&ColorFromCurve(ColorCurve=DataInterface() {propertyValues {
       AlphaCurve {keys [<at 0.5723 0 RCIM_Cubic>,<at 0.6433 1 RCIM_Cubic>]}
       BlueCurve {keys [<at 0 1 RCIM_Cubic>,<at 1 1 RCIM_Cubic>]}
       GreenCurve {keys [<at 0 1 RCIM_Cubic>,<at 1 1 RCIM_Cubic>]}
       RedCurve {keys [<at 0 1 RCIM_Cubic>,<at 1 1 RCIM_Cubic>]}
      }})) {isSetParameters=true}
   &Color(Color=&ColorFromCurve(ColorCurve=DataInterface() {propertyValues {
       AlphaCurve {keys [
         <at 0.0106 0>
         <at 0.069 1>
         {arriveTangent=-1.657, interpMode=RCIM_Cubic, leaveTangent=-1.657, time=0.5807, value=1}
         <at 0.6725 0 RCIM_Cubic>
        ]}
       BlueCurve {keys [<at 0.0774 1 RCIM_Cubic>,<at 1 1 RCIM_Cubic>]}
       GreenCurve {keys [<at 0.0774 1 RCIM_Cubic>,<at 1 1 RCIM_Cubic>]}
       RedCurve {keys [<at 0.0774 1 RCIM_Cubic>,<at 1 1 RCIM_Cubic>]}
      }}))
   &ScaleSpriteSize(Scale Factor=&Vector2DFromCurve(
     Vector2Curve=DataInterface() {propertyValues {XCurve {keys [<at 0 0.9>,<at 0.7052 0.3>]}, YCurve {keys [<at -0.0001 0.2829>,<at 0.6407 1.5>]}}}
     ScaleCurve=<1 1>
    )) {module=ScaleSpriteSize001}
  ]
  rendererClasses [&NiagaraSpriteRendererProperties]
  renderers [
   {rendererClass=&NiagaraSpriteRendererProperties}
   {rendererClass=&NiagaraSpriteRendererProperties, rendererIndex=1}
  ]
 }}
summary {
 systemName=NS_Custodial_SweatDrips
 emitters [{emitterName=Drip1, rendererClasses [&NiagaraSpriteRendererProperties]}]
 userVariables [User.SpriteRotation {type=NiagaraFloat, defaultValue=0, description=""}]
}
